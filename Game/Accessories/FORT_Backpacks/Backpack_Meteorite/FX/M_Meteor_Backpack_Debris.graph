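# Unlit masked material for the Meteorite backpack's rock debris, expanding geometry into camera-facing sprites that shrink near the camera.
# Vertex red scales size, brightness and motion, green varies rotation and noise phase, and blue drives emissive, joined by a sphere-mask glow around an offset actor position.
# Opacity mask multiplies rock alpha by a noise-distorted edge falloff.
# Optional panning world-space noise adds drift, before Creative move-tool and emissive-adjustment functions.
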Material /Game/Accessories/FORT_Backpacks/Backpack_Meteorite/FX/M_Meteor_Backpack_Debris
graph {
 properties {
  bUseMaterialAttributes=true
  blendMode=BLEND_Masked
  ditheredLODTransition=true
  pixelDepthOffsetMode=PDOM_Legacy
  shadingModel=MSM_Unlit
  twoSided=true
  usedShadingModels=Unlit
 }
 nodes [
  n1 = VectorParameter() {parameterName=Debris Color Mult, defaultValue=<rgba 0 0 0 1>}
  n2 = VertexColor()
  n3 = ScalarParameter() {defaultValue=1.05, parameterName=Zoom}
  n4 = ScalarParameter() {defaultValue=12, parameterName=Rot multiplier}
  n5 = TextureSample(UVs=Rotator(
    Coordinate=MaterialFunctionCall(Texture Scale=@n3) {materialFunction=&ScaleUVsByCenter}.UVs
    Time=Multiply(A=@n2.G, B=@n4)
   )) {texture=/Game/VisualThreatManager/Rifts/Textures/T_RocksSmall_02}
  n6 = VectorParameter() {parameterName=Emissive Color, defaultValue=<rgba 1 0.2995 0 1>}
  n7 = Multiply(A=Power(Base=VertexColor().B, Exp=ScalarParameter() {parameterName=Emissive Color Mask Power}), B=@n6.RGB)
  n8 = VectorParameter() {group=Glow, parameterName=Glow Offset, defaultValue=<rgba 0 0 0 1>}
  n9 = SphereMask(
   A=WorldPosition().XYZ
   B=Add(A=ActorPositionWS(), B=Transform(None=@n8.RGB) {transformSourceType=TRANSFORMSOURCE_Local})
   Radius=ScalarParameter() {defaultValue=25, group=Glow, parameterName=Glow Radius}
   Hardness=ScalarParameter() {defaultValue=25, group=Glow, parameterName=Glow Hardness}
  )
  n10 = Add(A=Multiply(A=Multiply(A=@n1.RGB, B=@n2.R), B=@n5.RGB), B=Add(A=Multiply(A=@n7) {constB=2}, B=LinearInterpolate(B=@n6.RGB, Alpha=@n9)))
  n11 = TextureCoordinate()
  n12 = TextureCoordinate() {uTiling=1.4, vTiling=1.4}
  n13 = Multiply(A=VertexColor().G, B=ScalarParameter() {defaultValue=12, parameterName=Rot multiplier})
  n14 = ScalarParameter() {defaultValue=0.15, parameterName=EdgeDistort}
  n15 = Multiply(
   A=TextureSample(UVs=Rotator(Coordinate=@n12, Time=@n13)) {samplerType=SAMPLERTYPE_LinearGrayscale, texture=/Game/Global/Textures/Default/Noise/T_SwirlNoise_N_height}.G
   B=@n14
  )
  n16 = Multiply(A=@n14) {constB=-1}
  n17 = MaterialFunctionCall(UV Channel=Add(A=Add(A=@n11, B=@n15), B=Divide(A=@n16))) {materialFunction=&LinearGradient}
  n18 = Multiply(A=@n17.UGradient, B=OneMinus(None=@n17.UGradient))
  n19 = Multiply(A=@n17.VGradient, B=OneMinus(None=@n17.VGradient))
  n20 = Multiply(A=Power(Base=Multiply(A=@n18, B=@n19)) {constExponent=1.4}) {constB=256}
  n21 = Subtract(A=Multiply(A=@n5.A, B=Saturate(None=@n20)), B=ScalarParameter() {parameterName=Subtract Alpha Edge})
  n22 = ScalarParameter() {defaultValue=0.05, parameterName=Sprite Size}
  n23 = Power(Base=VertexColor().R, Exp=ScalarParameter() {defaultValue=1, parameterName=SpriteSizePower})
  n24 = ScalarParameter() {defaultValue=50, parameterName=NearCamShrinkDist}
  n25 = MaterialFunctionCall!(
   Center Location=WorldPosition().XYZ
   0-1 UVs
   XY scale=Multiply(
    A=Multiply(A=@n22, B=@n23)
    B=MaterialFunctionCall(Distance=@n24) {materialFunction=&Near_Camera_Fade}.Result
   )
   OverridePivot
   "Normals (Optional)"
  ) {materialFunction=&Sprite}
  n26 = VertexColor()
  n27 = Add(A=Multiply(A=WorldPosition().XYZ) {constB=0.001}, B=@n26.G)
  n28 = Add(A=Time(), B=@n26.G)
  n29 = AppendVector(A=ScalarParameter() {defaultValue=0.02, parameterName=WPO Speed X}, B=ScalarParameter() {defaultValue=0.065, parameterName=WPO Speed Y})
  n30 = Add(A=ComponentMask(None=@n27) {g=true, r=true}, B=Multiply(A=@n28, B=@n29))
  n31 = Add(A=ComponentMask(None=@n27) {b=true, g=true}, B=Multiply(A=@n28, B=@n29))
  n32 = Multiply(
   A=Add(
    A=TextureSample(UVs=@n30) {samplerType=SAMPLERTYPE_LinearColor, texture=&HighContrastNoise}.RGB
    B=TextureSample(UVs=@n31) {samplerType=SAMPLERTYPE_LinearColor, texture=&HighContrastNoise}.RGB
   )
   B=@n26.R
  )
  n33 = ScalarParameter() {defaultValue=18, parameterName=Movement Intensity}
  n34 = StaticSwitchParameter(True=Add(A=@n25.World position offset, B=Multiply(A=@n32, B=@n33)), False=@n25.World position offset) {parameterName=Has WPO Movement?}
  MaterialFunctionCall!(In=MaterialFunctionCall(MA In=MakeMaterialAttributes!(
     BaseColor
     Metallic
     Specular
     Roughness
     Anisotropy
     EmissiveColor=@n10
     Opacity
     OpacityMask=@n21
     Normal
     Tangent
     WorldPositionOffset=@n34
     SubsurfaceColor
     ClearCoat
     ClearCoatRoughness
     AmbientOcclusion
     Refraction
     CustomizedUVs_0
     CustomizedUVs_1
     CustomizedUVs_2
     CustomizedUVs_3
     CustomizedUVs_4
     CustomizedUVs_5
     CustomizedUVs_6
     CustomizedUVs_7
     PixelDepthOffset
     ShadingModel
     Displacement
    )) {materialFunction=&MF_CreativeMoveTool}.MA Out) {materialFunction=/Game/Effects/Niagara/MasterMaterials/Functions/MF_FX_EmissiveAdjustments_MA}
 ]
}
